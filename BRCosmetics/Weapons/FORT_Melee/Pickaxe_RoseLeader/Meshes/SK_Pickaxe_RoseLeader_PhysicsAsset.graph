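# Physics asset for the RoseLeader two-handed melee pickaxe skeletal mesh, a cosmetic weapon.
# It has a single capsule body on the Melee_TwoHanded_Handle bone, a coarse volume offset up the local Z axis and rotated a quarter turn.
# It has no constraints, so the pickaxe has no jointed or secondary simulated parts.
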PhysicsAsset /BRCosmetics/Weapons/FORT_Melee/Pickaxe_RoseLeader/Meshes/SK_Pickaxe_RoseLeader_PhysicsAsset
bodies {Melee_TwoHanded_Handle [{shapeType=Capsule, length=69.4996, radius=40.6337, center=<0.6875 0.4218 33.7276>, rotation=<rot -0.3046 -90 0.1288>}]}
constraints []
